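# Sound cue for the destruction of the WildEstate factory chest, playing a single wave in the impacts sound class.
# The wave feeds a modulator that randomizes pitch while holding volume fixed, then passes through a one-input mixer to a boosted output.
# Spatialization uses a shared small-melee-weapon attenuation asset; the stored occlusion and falloff overrides are not switched on.
# A concurrency override blocks new instances once the limit is reached.
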SoundCue /WildEstate_FactionChest/Audio/WildEstate_FactoryChest_Destroyed_Cue
root=Mixer_1
nodes [
 {id=Mixer_1, children [Modulator_1], properties {inputVolume [1]}}
 {id=Modulator_1, children [WavePlayer_1], properties {pitchMax=1.1, volumeMax=1, volumeMin=1}}
 {id=WavePlayer_1, properties {soundWaveAssetPtr=/WildEstate_FactionChest/Audio/Waves/WildEstate_FactoryChest_Destroyed}}
]
properties {
 attenuationSettings=&Fort_Weapon_Melee_Sml_Att
 bOverrideConcurrency=true
 maxDistance=2800
 soundClassObject=&Fort_Impacts
 volumeMultiplier=1.5
 attenuationOverrides {bEnableOcclusion=true, falloffDistance=1000, occlusionTraceChannel=ECC_GameTraceChannel11, attenuationShapeExtents {x=500}}
 concurrencyOverrides {resolutionRule=PreventNew}
}
