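# Sound cue for impacts on the WildEstate faction chest, named FactoryChest in the asset, routed to the Fort impacts sound class.
# A Random node picks one of six equally weighted impact waves, and a Modulator varies pitch at fixed volume before a single-input Mixer.
# The cue plays at reduced volume with small melee weapon attenuation, and its own concurrency override rejects new instances once the limit is reached.
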SoundCue /WildEstate_FactionChest/Audio/WildEstate_FactoryChest_Impact_Cue
root=Mixer_1
nodes [
 {id=Mixer_1, children [Modulator_1], properties {inputVolume [1]}}
 {id=Modulator_1, children [Random_1], properties {pitchMax=1.1, volumeMax=1, volumeMin=1}}
 {id=Random_1, children [WavePlayer_0,WavePlayer_4,WavePlayer_5,WavePlayer_6,WavePlayer_7,WavePlayer_8], properties {weights [1,1,1,1,1,1]}}
 {id=WavePlayer_0, properties {soundWaveAssetPtr=/WildEstate_FactionChest/Audio/Waves/WildEstate_FactoryChest_Impact_01}}
 {id=WavePlayer_4, properties {soundWaveAssetPtr=/WildEstate_FactionChest/Audio/Waves/WildEstate_FactoryChest_Impact_02}}
 {id=WavePlayer_5, properties {soundWaveAssetPtr=/WildEstate_FactionChest/Audio/Waves/WildEstate_FactoryChest_Impact_03}}
 {id=WavePlayer_6, properties {soundWaveAssetPtr=/WildEstate_FactionChest/Audio/Waves/WildEstate_FactoryChest_Impact_04}}
 {id=WavePlayer_7, properties {soundWaveAssetPtr=/WildEstate_FactionChest/Audio/Waves/WildEstate_FactoryChest_Impact_05}}
 {id=WavePlayer_8, properties {soundWaveAssetPtr=/WildEstate_FactionChest/Audio/Waves/WildEstate_FactoryChest_Impact_06}}
]
properties {
 attenuationSettings=&Fort_Weapon_Melee_Sml_Att
 bOverrideConcurrency=true
 maxDistance=2800
 soundClassObject=&Fort_Impacts
 volumeMultiplier=0.5
 attenuationOverrides {bEnableOcclusion=true, falloffDistance=1000, occlusionTraceChannel=ECC_GameTraceChannel11, attenuationShapeExtents {x=500}}
 concurrencyOverrides {resolutionRule=PreventNew}
}
