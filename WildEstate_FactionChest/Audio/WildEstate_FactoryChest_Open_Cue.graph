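# Sound cue for opening the WildEstate factory chest, routed to the loot reward sound class.
# A Random node picks one of three equally weighted open waves, then a Modulator varies its pitch and passes it through a single-input Mixer.
# It references the shared item-interact attenuation asset for spatialization.
# A concurrency override rejects new instances at the limit rather than stopping playing ones.
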SoundCue /WildEstate_FactionChest/Audio/WildEstate_FactoryChest_Open_Cue
root=Mixer_1
nodes [
 {id=Mixer_1, children [Modulator_1], properties {inputVolume [1]}}
 {id=Modulator_1, children [Random_0], properties {pitchMax=1.1, volumeMax=1, volumeMin=1}}
 {id=Random_0, children [WavePlayer_1,WavePlayer_2,WavePlayer_3], properties {weights [1,1,1]}}
 {id=WavePlayer_1, properties {soundWaveAssetPtr=/WildEstate_FactionChest/Audio/Waves/WildEstate_FactoryChest_Open_01}}
 {id=WavePlayer_2, properties {soundWaveAssetPtr=/WildEstate_FactionChest/Audio/Waves/WildEstate_FactoryChest_Open_02}}
 {id=WavePlayer_3, properties {soundWaveAssetPtr=/WildEstate_FactionChest/Audio/Waves/WildEstate_FactoryChest_Open_03}}
]
properties {
 attenuationSettings=/Game/Sounds/Attenuation/Fort_Item_Interact_Att
 bOverrideConcurrency=true
 maxDistance=1850
 soundClassObject=&Fort_Loot_Reward_PSM
 volumeMultiplier=1.5
 attenuationOverrides {bEnableOcclusion=true, falloffDistance=1000, occlusionTraceChannel=ECC_GameTraceChannel11, attenuationShapeExtents {x=500}}
 concurrencyOverrides {resolutionRule=PreventNew}
}
