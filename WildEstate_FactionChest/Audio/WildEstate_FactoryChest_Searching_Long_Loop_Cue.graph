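# Looping sound cue for searching a Wild Estate chest, built from shared small cardboard box foley.
# A mixer layers a random one of three one-shot start sounds with a random one of three looping search sounds, the loop at lower volume.
# Each layer passes through its own modulator for pitch variation.
# It uses item-interact attenuation, the loot-collect sound class and a concurrency override that prevents new instances at the limit.
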SoundCue /WildEstate_FactionChest/Audio/WildEstate_FactoryChest_Searching_Long_Loop_Cue
root=Mixer_1
nodes [
 {id=Mixer_1, children [Modulator_4,Modulator_3], properties {inputVolume [1,0.7]}}
 {id=Modulator_3, children [Random_2], properties {pitchMax=1.2, pitchMin=0.9, volumeMax=1, volumeMin=1}}
 {id=Modulator_4, children [Random_3], properties {pitchMax=1.5, pitchMin=0.9}}
 {id=Random_2, children [WavePlayer_6,WavePlayer_7,WavePlayer_8], properties {bShouldExcludeFromBranchCulling=true, weights [1,1,1]}}
 {id=Random_3, children [WavePlayer_9, WavePlayer_10, WavePlayer_11], properties {bShouldExcludeFromBranchCulling=true, weights [1,1,1]}}
 {id=WavePlayer_10, properties {soundWaveAssetPtr=/Game/Sounds/Foley_Loot/Containers/CardboardBox/CardboardBox_Small_Searching_Loop_Start_02}}
 {id=WavePlayer_11, properties {soundWaveAssetPtr=/Game/Sounds/Foley_Loot/Containers/CardboardBox/CardboardBox_Small_Searching_Loop_Start_03}}
 {
  id=WavePlayer_6
  properties {bLooping=true, soundWaveAssetPtr=/Game/Sounds/Foley_Loot/Containers/CardboardBox/CardboardBox_Small_Searching_Loop_01}
 }
 {
  id=WavePlayer_7
  properties {bLooping=true, soundWaveAssetPtr=/Game/Sounds/Foley_Loot/Containers/CardboardBox/CardboardBox_Small_Searching_Loop_02}
 }
 {
  id=WavePlayer_8
  properties {bLooping=true, soundWaveAssetPtr=/Game/Sounds/Foley_Loot/Containers/CardboardBox/CardboardBox_Small_Searching_Loop_03}
 }
 {id=WavePlayer_9, properties {soundWaveAssetPtr=/Game/Sounds/Foley_Loot/Containers/CardboardBox/CardboardBox_Small_Searching_Loop_Start_01}}
]
properties {
 attenuationSettings=/Game/Sounds/Attenuation/Fort_Item_Interact_Mono_Att
 bOverrideConcurrency=true
 maxDistance=1500
 soundClassObject=&Fort_Loot_Collect
 volumeMultiplier=0.8141
 concurrencyOverrides {resolutionRule=PreventNew}
}
